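; Cosmetic display stand blueprint that swaps an AuraCop outfit set between its default and white (BulletBlanc) variants.
; When the first player's pawn begins overlapping the Capsule, a FlipFlop alternates between the two variants, starting with white.
; Each branch sets slot 0 materials on glider, pickaxe, character, accessory and backpack, then updates the variant-name TextRender.
; Material paths are hard-coded; BeginPlay, Tick, actor overlap and the construction script are empty.
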(event EventBeginPlay)

(event Collision|EventActorBeginOverlap
  (params (OtherActor :type Actor%20Object%20Reference)))

(event EventTick
  (params (DeltaSeconds :type Float%20%28single-precision%29)))

(event OnComponentBeginOverlap(Capsule)
  (params (OverlappedComponent :type Primitive%20Component%20Object%20Reference) (OtherActor :type Actor%20Object%20Reference) (OtherComp :type Primitive%20Component%20Object%20Reference) (OtherBodyIndex :type Integer) (bFromSweep :type Boolean) (SweepResult :type Hit%20Result%20Structure%20%28by%20ref%29))
  (Variables|Default|SetOther%20Actor OtherActor)
  (if (== (Variables|Default|GetOther%20Actor) (Game|GetPlayerPawn 0))
    (Utilities|FlowControl|FlipFlop
      (:A
        ;[1 1 1] AuraCop White
        (Rendering|Material|SetMaterial (Variables|Default|GetGlider) 0 "/BRCosmetics/Gadgets/Assets/VinderTech_GliderChute/Glider_AuraCop/Skins/BulletBlanc/Materials/MI_AuraCop_BulletBlanc_Glider.MI_AuraCop_BulletBlanc_Glider")
        (Rendering|Material|SetMaterial (Variables|Default|GetPickaxe) 0 "/BRCosmetics/Weapons/FORT_Melee/Pickaxe_AuraCop/Skins/BulletBlanc/Materials/MI_Pickaxe_AuraCop_BulletBlanc.MI_Pickaxe_AuraCop_BulletBlanc")
        (Rendering|Material|SetMaterial (Variables|Default|GetCharacterAccessory) 0 "/BRCosmetics/Characters/Player/Male/Medium/Bodies/M_MED_AuraCop_Spur/Skins/Blanc/Materials/MI_AuraCop_SpurBlanc_FaceAcc.MI_AuraCop_SpurBlanc_FaceAcc")
        (Rendering|Material|SetMaterial (Variables|Default|GetCharacter) 0 "/BRCosmetics/Characters/Player/Male/Medium/Bodies/M_MED_AuraCop_Bullet/Skins/Blanc/Materials/MI_AuraCop_BulletBlanc_Body.MI_AuraCop_BulletBlanc_Body")
        (Rendering|Material|SetMaterial (Variables|Default|GetBackpack) 0 "/BRCosmetics/Accessories/FORT_Backpacks/Backpack_AuraCop/Skins/BulletBlanc/Materials/MI_Backpack_AuraCop_BulletBlanc.MI_Backpack_AuraCop_BulletBlanc")
        (Rendering|Components|TextRender|SetText (Variables|Default|GetTextRender(VARIANT_NAME)) (Variables|Default|GetAuraCop_White)))
      (:B
        ;[1 1 1] AuraCop Default
        (Rendering|Material|SetMaterial (Variables|Default|GetGlider) 0 "/BRCosmetics/Gadgets/Assets/VinderTech_GliderChute/Glider_AuraCop/Materials/MI_AuraCop_Glider.MI_AuraCop_Glider")
        (Rendering|Material|SetMaterial (Variables|Default|GetPickaxe) 0 "/BRCosmetics/Weapons/FORT_Melee/Pickaxe_AuraCop/Materials/MI_Pickaxe_AuraCop.MI_Pickaxe_AuraCop")
        (Rendering|Material|SetMaterial (Variables|Default|GetCharacterAccessory) 0 "/BRCosmetics/Characters/Player/Male/Medium/Bodies/M_MED_AuraCop_Spur/Materials/MI_AuraCop_Spur_FaceAcc.MI_AuraCop_Spur_FaceAcc")
        (Rendering|Material|SetMaterial (Variables|Default|GetCharacter) 0 "/BRCosmetics/Characters/Player/Male/Medium/Bodies/M_MED_AuraCop_Bullet/Materials/MI_AuraCop_Bullet_Body.MI_AuraCop_Bullet_Body")
        (Rendering|Material|SetMaterial (Variables|Default|GetBackpack) 0 "/BRCosmetics/Accessories/FORT_Backpacks/Backpack_AuraCop/Materials/MI_Backpack_AuraCop.MI_Backpack_AuraCop")
        (Rendering|Components|TextRender|SetText (Variables|Default|GetTextRender(VARIANT_NAME)) (Variables|Default|GetAuraCop))))))

(fn ConstructionScript)
